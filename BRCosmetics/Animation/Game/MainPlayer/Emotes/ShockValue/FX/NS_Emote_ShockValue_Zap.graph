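# Niagara effect for the ShockValue emote's zap: electric arcs, sparks and smoke sampled from the character's skeletal mesh.
# Ramping GPU mesh lightning cycles through fifteen bone/socket indices, while another single index anchors spikes, sprite bolts, smoke and a flickering flare with light.
# A delayed smoke burst covers more bones, and a second sprite-lightning emitter is disabled.
# Time-of-day and front-end inputs tune brightness and tint; User.IsFollower gates the flare.
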NiagaraSystem /BRCosmetics/Animation/Game/MainPlayer/Emotes/ShockValue/FX/NS_Emote_ShockValue_Zap
dependencies {
 usedDataInterfaces [
  &NiagaraDataInterfaceCurve
  &NiagaraDataInterfaceArrayInt32
  &NiagaraDataInterfaceColorCurve
  &NiagaraDataInterfaceVector2DCurve
  &NiagaraDataInterfaceSpriteRendererInfo
  &NiagaraDataInterfaceSkeletalMesh
 ]
 usedDynamicInputs [
  &Add_Vector
  /Niagara/DynamicInputs/Helpers/SimulationPosition
  &MaskFloat
  &UniformRangedVector
  &Vector2DFromCurve
 ]
 usedModules [
  &SystemState
  &self:SystemSpawnScript.SystemScriptSource.SystemScriptGraph.NiagaraNodeAssignment_0.SetVariables_#1
  &self:SystemSpawnScript.SystemScriptSource.SystemScriptGraph.NiagaraNodeAssignment_2.SetVariables_#2
 ]
}
emitters {
 Impact_Spikes {
  emitterUpdateScript [
   &EmitterState(Life Cycle Mode=::NewEnumerator1, Loop Behavior=::NewEnumerator1, Loop Duration=1, UseLoopDelay=false)
   &SpawnBurst_Instantaneous(Spawn Count=13, Use Spawn Probability=false, UseLoopCountLimit=false)
  ]
  particleSpawnScript [
   &InitializeParticle(
    Color=<rgba 20 18.1825 10.5671 1>
    Lifetime Max=0.1
    Lifetime Min=0.12
    Lifetime Mode=::NewEnumerator1
    Sprite Size=<5 30>
    Sprite Size Mode=::NewEnumerator4
    UsePositionOffset=false
    Write Material Random=true
   )
   &SkeletalMeshLocation(
    Bone / Socket Index=&SelectIntFromArray(Direct Array Index=&ReturnExecIndex(), Int Selection Array=DataInterface() {propertyValues {IntData [16]}})
    Bone Sampling Mode=::NewEnumerator4
    MaskSpawnGroup=false
    Position Sampling=::NewEnumerator1
    Skeletal Mesh=$System.Skeletal Mesh
   )
   &SphereLocation(
    MaskSpawnGroup=false
    Override Seed=false
    Sphere Radius=1
    Surface Only=true
    UseNonUniformScale=false
    UseOffset=true
    UseSphereOrientation=false
   )
   &SubUVAnimation(End Frame=3, SubUV Animation Mode:ENiagara_SubUVLookupMode=::NewEnumerator2)
   &AddVelocity(Velocity=&Multiply_Vector(
     A=$Output.SphereLocation.SphereNormal
     B=&VectorFromFloat(Value=&RandomRangeFloat(Minimum=50, Maximum=80))
    )) {module=AddVelocity001}
  ]
  particleUpdateScript [
   &ParticleState()
   &Drag(Drag=3, Use Linear Drag=true, Use Rotational Drag=false)
   &SolveForcesAndVelocity(Clamp Velocity=false, Limit Acceleration=false)
   &ScaleSpriteSize(
    Non-Uniform Curve Scale=&Multiply_Vector2D_ByFloat(Float=0.8, Vector2D=<1 1>)
    Non-Uniform Curve Sprite Scale=DataInterface() {propertyValues {
      XCurve {keys [{tangentMode=RCTM_User, time=0, value=1}, {tangentMode=RCTM_User, time=1, value=0.8}]}
      YCurve {keys [{tangentMode=RCTM_User, time=0, value=0.9868}, {tangentMode=RCTM_User, time=1, value=1.7}]}
     }}
    Scale Sprite Size Mode=::NewEnumerator3
   )
   &ScaleColor(
    Scale Alpha=&Multiply_Float(
     A=&FloatFromCurve(
      FloatCurve=DataInterface() {propertyValues {Curve {keys [<at 0 0>,<at 0.1 1>,<at 0.2 0.8>,<at 1 0>]}}}
      Scale Curve=&FNTimeOfDayScalar(Night Scale=1, Morning Scale=0.3, Evening Scale=0.5)
     )
     B=0.1
    )
    Scale RGB=<1 0.8 0.8>
    ScaleA=true
    ScaleRGB=true
   )
   /SetVariables(Particles.Color2=&MultiplyLinearColor(Linear Color=$Particles.Color, Scale Factor=<1 1 1 0.2>)) {isSetParameters=true}
  ]
  rendererClasses [&NiagaraSpriteRendererProperties]
  renderers [{rendererClass=&NiagaraSpriteRendererProperties}]
 }
 Lightning_Small_L001 {
  emitterUpdateScript [
   &EmitterState(Life Cycle Mode=::NewEnumerator1, Loop Behavior=::NewEnumerator1, Loop Delay=0.55, Loop Duration=4, UseLoopDelay=false)
   &SpawnBurst_Instantaneous(Spawn Count=5, Spawn Probability=0.7, Use Spawn Probability=false, UseLoopCountLimit=false)
   &SpawnRate(SpawnRate=7, Use Spawn Probability=false)
  ]
  particleSpawnScript [
   &PartitionParticles()
   &InitializeParticle(
    Color=<rgba 5 4.0321 1.276 1>
    Lifetime Max=0.15
    Lifetime Min=0.2
    Lifetime Mode=::NewEnumerator1
    Mesh Scale=&Multiply_VectorByFloat(Vector=&RandomRangeVector(Minimum=<0.2 0.2 1.5>, Maximum=<0.2 0.2 2>), Float=0.2)
    Mesh Scale Mode=::NewEnumerator4
    Sprite Size=&Multiply_Vector2D_ByFloat(Float=23.7691, Vector2D=&RandomRangeVector2D(Minimum=<0.3 2>, Maximum=<0.5 1.5>))
    Sprite Size Mode=::NewEnumerator4
    UsePositionOffset=false
    Write Material Random=true
   )
   &CameraOffset(Camera Offset Amount=20)
   &SkeletalMeshLocation(
    Bone / Socket Index=&SelectIntFromArray(Direct Array Index=&ReturnExecIndex(), Int Selection Array=DataInterface() {propertyValues {IntData [16]}})
    Bone Sampling Mode=::NewEnumerator4
    MaskSpawnGroup=false
    Skeletal Mesh=$System.Skeletal Mesh
   )
   &ShapeLocation(MaskSpawnGroup=false, Override Seed=false, Sphere Radius=5, Sphere Surface Distribution=1, UseOwnerScale=true)
   &AddVelocity(
    Module.Use Falloff Distance=false
    Override Seed=false
    Velocity Mode=::NewEnumerator1
    Velocity Origin=$Particles.SkeletalMeshLocation.SampledPosition
    Velocity Speed=1
   )
   &SubUVAnimation(
    End Frame Range Override=2
    Evaluation Type=::NewEnumerator0
    "Sprite Renderer "=DataInterface() {propertyValues {SpriteRenderer=&self:Lightning_Small_L001_0.NiagaraSpriteRendererProperties_0}}
    SubUV Animation Mode:ENiagara_SubUVLookupModeV2=::NewEnumerator2
    UseEndFrame=true
    UseStartFrame=true
   )
  ]
  particleUpdateScript [
   &ParticleState()
   &ScaleSpriteSize(
    Non-Uniform Curve Sprite Scale=DataInterface() {propertyValues {
      XCurve {keys [{tangentMode=RCTM_User, time=0, value=1}, {tangentMode=RCTM_User, time=1, value=1}]}
      YCurve {keys [
        {arriveTangent=1.2594, interpMode=RCIM_Cubic, leaveTangent=1.2594, tangentMode=RCTM_User, time=0, value=0.4276}
        {interpMode=RCIM_Cubic, tangentMode=RCTM_User, time=1, value=1}
       ]}
     }}
    Scale Sprite Size Mode=::NewEnumerator3
   )
   &SolveForcesAndVelocity(Clamp Velocity=false, Limit Acceleration=false)
   &ScaleColor(
    Scale Alpha=&Multiply_Float(
     A=20
     B=&FloatFromCurve(FloatCurve=DataInterface() {propertyValues {Curve {keys [
          {tangentMode=RCTM_User, time=0, value=1}
          {arriveTangent=-0.0405, interpMode=RCIM_Cubic, leaveTangent=-0.0405, tangentMode=RCTM_User, time=0.7009, value=1}
          {arriveTangent=-7.2455, arriveTangentWeight=0.8048, interpMode=RCIM_Cubic, leaveTangent=-7.2455, tangentMode=RCTM_User, time=1, value=0}
         ]}}})
    )
    Scale RGB=&Multiply_VectorByFloat(
     Vector=<1 1 1>
     Float=&Multiply_Float(A=&FloatFromCurve(FloatCurve=DataInterface() {propertyValues {Curve {keys [<at 0 2>,<at 1 2>]}}}), B=1)
    )
    ScaleA=true
    ScaleRGB=true
   )
  ]
  rendererClasses [&NiagaraSpriteRendererProperties]
  renderers [{rendererClass=&NiagaraSpriteRendererProperties}]
 }
 Lightning_Small_L002 {
  bEnabled=false
  emitterUpdateScript [
   &EmitterState(Life Cycle Mode=::NewEnumerator1, Loop Behavior=::NewEnumerator1, Loop Delay=0.55, Loop Duration=4, UseLoopDelay=false)
   &SpawnBurst_Instantaneous(Spawn Count=5, Spawn Time=2, Use Spawn Probability=false, UseLoopCountLimit=false) {module=SpawnBurst_Instantaneous003}
   &SpawnBurst_Instantaneous(Spawn Count=6, Spawn Time=1.2, Use Spawn Probability=false, UseLoopCountLimit=false) {module=SpawnBurst_Instantaneous002}
   &SpawnBurst_Instantaneous(Spawn Count=5, Spawn Time=0.4, Use Spawn Probability=false, UseLoopCountLimit=false) {module=SpawnBurst_Instantaneous001}
   &SpawnBurst_Instantaneous(Spawn Count=5, Spawn Probability=0.7, Spawn Time=0.05, Use Spawn Probability=false, UseLoopCountLimit=false)
  ]
  particleSpawnScript [
   &PartitionParticles()
   &InitializeParticle(
    Color=<rgba 5 4.0321 1.276 1>
    Lifetime Max=0.15
    Lifetime Min=0.07
    Lifetime Mode=::NewEnumerator1
    Mesh Scale=&Multiply_VectorByFloat(Vector=&RandomRangeVector(Minimum=<0.2 0.2 1.5>, Maximum=<0.2 0.2 2>), Float=0.2)
    Mesh Scale Mode=::NewEnumerator4
    Sprite Size=&Multiply_Vector2D_ByFloat(Float=23.7691, Vector2D=&RandomRangeVector2D(Minimum=<0.3 2>, Maximum=<0.5 1.5>))
    Sprite Size Mode=::NewEnumerator4
    UsePositionOffset=false
    Write Material Random=true
   )
   &CameraOffset(Camera Offset Amount=20)
   &SkeletalMeshLocation(
    Bone / Socket Index=&SelectIntFromArray(Direct Array Index=$Particles.PartitionParticles.Partition, Int Selection Array=DataInterface() {propertyValues {IntData [9,17]}})
    Bone Sampling Mode=::NewEnumerator4
    MaskSpawnGroup=false
    Position Sampling=::NewEnumerator1
    Sampled Position Offset=&MaskVector(Mask Parameter=$Particles.PartitionParticles.Partition, Mask=0, Masked Vector=<-10 -5 0>, Passthrough Vector=<0 0 10>)
    Skeletal Mesh=$System.Skeletal Mesh
   )
   &ShapeLocation(MaskSpawnGroup=false, Override Seed=false, Sphere Radius=5, Sphere Surface Distribution=1, UseOwnerScale=true)
   &AddVelocity(
    Cone Angle=117.715
    Cone Axis=&MaskVector(Mask Parameter=$Particles.PartitionParticles.Partition, Mask=0, Masked Vector=<-59.6404 0 -1>, Passthrough Vector=<1 0 -167.0207>)
    Module.UseConeFalloff=false
    Module.UseVelDistribution=false
    Override Seed=false
    Velocity Mode=::NewEnumerator2
    Velocity Speed=1.1248
   )
   &SubUVAnimation(
    End Frame Range Override=2
    Evaluation Type=::NewEnumerator0
    "Sprite Renderer "=DataInterface() {propertyValues {SpriteRenderer=&self:Lightning_Small_L002_0.NiagaraSpriteRendererProperties_0}}
    SubUV Animation Mode:ENiagara_SubUVLookupModeV2=::NewEnumerator2
    UseEndFrame=true
    UseStartFrame=true
   )
  ]
  particleUpdateScript [
   &ParticleState()
   &ScaleSpriteSize(
    Non-Uniform Curve Sprite Scale=DataInterface() {propertyValues {
      XCurve {keys [{tangentMode=RCTM_User, time=0, value=1}, {tangentMode=RCTM_User, time=1, value=1}]}
      YCurve {keys [
        {arriveTangent=1.2594, interpMode=RCIM_Cubic, leaveTangent=1.2594, tangentMode=RCTM_User, time=0, value=0.4276}
        {interpMode=RCIM_Cubic, tangentMode=RCTM_User, time=1, value=1}
       ]}
     }}
    Scale Sprite Size Mode=::NewEnumerator3
   )
   &SolveForcesAndVelocity(Clamp Velocity=false, Limit Acceleration=false)
   &ScaleColor(
    Scale Alpha=&Multiply_Float(
     A=20
     B=&FloatFromCurve(FloatCurve=DataInterface() {propertyValues {Curve {keys [
          {tangentMode=RCTM_User, time=0, value=1}
          {arriveTangent=-0.0405, interpMode=RCIM_Cubic, leaveTangent=-0.0405, tangentMode=RCTM_User, time=0.7009, value=1}
          {arriveTangent=-7.2455, arriveTangentWeight=0.8048, interpMode=RCIM_Cubic, leaveTangent=-7.2455, tangentMode=RCTM_User, time=1, value=0}
         ]}}})
    )
    Scale RGB=&Multiply_VectorByFloat(
     Vector=<1 1 1>
     Float=&Multiply_Float(A=&FloatFromCurve(FloatCurve=DataInterface() {propertyValues {Curve {keys [<at 0 2>,<at 1 2>]}}}), B=1)
    )
    ScaleA=true
    ScaleRGB=true
   )
  ]
  rendererClasses [&NiagaraSpriteRendererProperties]
  renderers [{rendererClass=&NiagaraSpriteRendererProperties}]
 }
 Small_Lightning {
  simTarget=GPUComputeSim
  emitterUpdateScript [
   &EmitterState(Life Cycle Mode=::NewEnumerator1, Loop Behavior=::NewEnumerator1, Loop Delay=0.1, Loop Duration=4, UseLoopDelay=true)
   &SpawnRate(
    SpawnRate=&FloatFromCurve(
     FloatCurve=DataInterface() {propertyValues {Curve {keys [{interpMode=RCIM_Cubic, tangentMode=RCTM_User, time=0, value=0.22}, {interpMode=RCIM_Cubic, tangentMode=RCTM_User, time=1, value=1}]}}}
     CurveIndex=$Emitter.NormalizedLoopAge
     Scale Curve=30
    )
    Use Spawn Probability=false
   )
  ]
  particleSpawnScript [
   &InitializeParticle(
    Color Maximum=<rgba 2.2264 5 4.6542 1>
    Color Minimum=<rgba 5 4.2427 2.0862 1>
    Color Mode=::NewEnumerator2
    Lifetime=0.2
    Mesh Scale Max=<1 0.5 0.5>
    Mesh Scale Min=<0.5 1.5 0.9>
    Mesh Scale Mode=::NewEnumerator2
    Position Offset=<0 0 20>
    Sprite Size Mode=::NewEnumerator3
    UsePositionOffset=false
    Write Material Random=true
   )
   &CameraOffset(Camera Offset Amount=&RandomRangeFloat(Minimum=20, Maximum=20))
   &InitialMeshOrientation(Mesh Orientation Mode=::NewEnumerator2, Rotation=<0 0.25 0>, Use Rotation Vector=true)
   &InitialMeshOrientation(
    Mesh Orientation Mode=::NewEnumerator2
    Rotation=&RandomRangeVector(Minimum=<0 0 0>, Maximum=<1 0 0>)
    Use Rotation Vector=true
   ) {module=InitialMeshOrientation001}
   &ShapeLocation(MaskSpawnGroup=false, Override Seed=false, Sphere Radius=25, Sphere Surface Distribution=1, UseOwnerScale=true) {enabled=false}
   &AddVelocity(Module.Use Falloff Distance=false, Override Seed=false, Velocity Mode=::NewEnumerator1, Velocity Speed=1)
   &PartitionParticles(Number Of Partitions=15, Particles Per Partition=1, Partition Mode:ENiagara_PartitionMode=::NewEnumerator1)
  ]
  particleUpdateScript [
   &ParticleState()
   &SkeletalMeshLocation(
    Bone / Socket Index=&SelectIntFromArray(
     Direct Array Index=$Particles.PartitionParticles.Partition
     Int Selection Array=DataInterface() {propertyValues {IntData [4,3,14,13,5,1,6,0,2,12,11,8,7,10,9]}}
    )
    Bone Sampling Mode=::NewEnumerator4
    MaskSpawnGroup=false
    Orientation Sampling=::NewEnumerator1
    Skeletal Mesh=$System.Skeletal Mesh
   )
   &DynamicMaterialParameters(
    Index 0 Param 1=&FloatFromCurve(FloatCurve=DataInterface() {propertyValues {Curve {keys [<at 0 0>,<at 0.7 1>]}}})
    Index 0 Param 2=0.7
    Index 0 Param 3=&FloatFromCurve(FloatCurve=DataInterface() {propertyValues {Curve {keys [
         {arriveTangent=0.2581, interpMode=RCIM_Cubic, leaveTangent=0.2581, tangentMode=RCTM_User, time=0, value=0}
         {arriveTangent=2.8205, interpMode=RCIM_Cubic, leaveTangent=2.8205, tangentMode=RCTM_User, time=1, value=1}
        ]}}})
    Index 0 Param 4=1.5
    Index 1 Param 1=&FloatFromCurve(FloatCurve=DataInterface() {propertyValues {Curve {keys [<at 0 1>,<at 1 0.3>]}}})
    Module.Index0Float0Write=true
    Module.Index0Float1Write=true
    Module.Index0Float2Write=true
    Module.Index0Float3Write=true
    Module.Index1Float0Write=true
    Module.Index1Float1Write=true
    Module.Index1Float2Write=true
    Module.Index1Float3Write=true
    Write Parameter Index 1=::NewEnumerator1
   ) {module=DynamicMaterialParameters001}
   /SetVariables(Particles.Glow Color=&MakeLinearColorFromVectorAndFloat(
     "Vector (RGB)"=&MakeVectorFromLinearColorRGB(Color=&RandomRangeLinearColor(Minimum=<rgba 1 0.8485 0.4172 1>, Maximum=<rgba 0.0958 0.6244 0.5585 1>))
     "Float (Alpha)"=&Multiply_Float(
      A=&FloatFromCurve(FloatCurve=DataInterface() {propertyValues {Curve {keys [
           {interpMode=RCIM_Cubic, tangentMode=RCTM_User, time=0, value=1}
           <at 0.7 1>
           {arriveTangent=-3.6307, interpMode=RCIM_Cubic, leaveTangent=-3.6307, tangentMode=RCTM_User, time=0.9, value=0}
          ]}}})
      B=4
     )
    )) {isSetParameters=true}
   /SetVariables(Particles.Glow Dy02:Vector4f=&MakeVector4(
     X=&FloatFromCurve(FloatCurve=DataInterface() {propertyValues {Curve {keys [<at 0 0>,<at 0.7 1>]}}})
     Y=0.7
     Z=0
     W=0
    )) {isSetParameters=true}
   &SolveForcesAndVelocity(Clamp Velocity=false, Limit Acceleration=false)
   &ScaleColor(
    Scale Alpha=&Multiply_Float(
     A=20
     B=&FloatFromCurve(FloatCurve=DataInterface() {propertyValues {Curve {keys [
          {tangentMode=RCTM_User, time=0, value=1}
          {arriveTangent=-0.0405, interpMode=RCIM_Cubic, leaveTangent=-0.0405, tangentMode=RCTM_User, time=0.7009, value=1}
          {arriveTangent=-8.0089, interpMode=RCIM_Cubic, leaveTangent=-8.0089, tangentMode=RCTM_User, time=1, value=0}
         ]}}})
    )
    Scale RGB=&MakeVectorFromLinearColorRGB(Color=&RandomRangeLinearColor(
      Minimum=&ColorFromCurve(ColorCurve=DataInterface() {propertyValues {
         AlphaCurve {keys [<at 0 1 RCIM_Cubic>,<at 1 0 RCIM_Cubic>]}
         BlueCurve {keys [<at 0 1 RCIM_Cubic>,<at 0.3126 0.0369>,<at 0.7374 0.1065>,<at 1 1 RCIM_Cubic>]}
         GreenCurve {keys [<at 0 1 RCIM_Cubic>,<at 0.3126 0.1598>,<at 0.7374 0.4618>,<at 1 1 RCIM_Cubic>]}
         RedCurve {keys [<at 0 1 RCIM_Cubic>,<at 0.3126 0.1985>,<at 0.7374 0.5735>,<at 1 1 RCIM_Cubic>]}
        }})
      Maximum=&ColorFromCurve(ColorCurve=DataInterface() {propertyValues {
         AlphaCurve {keys [<at 0 1 RCIM_Cubic>,<at 1 0 RCIM_Cubic>]}
         BlueCurve {keys [<at 0 1 RCIM_Cubic>,<at 0.3126 0.15>,<at 0.7374 0.408>,<at 1 1 RCIM_Cubic>]}
         GreenCurve {keys [<at 0 1 RCIM_Cubic>,<at 0.3126 0.1985>,<at 0.7374 0.5735>,<at 1 1 RCIM_Cubic>]}
         RedCurve {keys [<at 0 1 RCIM_Cubic>,<at 0.3126 0>,<at 0.7374 0.0508>,<at 1 1 RCIM_Cubic>]}
        }})
     ))
    ScaleA=true
    ScaleRGB=true
   )
  ]
  rendererClasses [&NiagaraMeshRendererProperties]
  renderers [
   {rendererClass=&NiagaraMeshRendererProperties, rendererIndex=0}
   {rendererClass=&NiagaraMeshRendererProperties, rendererIndex=1}
  ]
 }
 SmokeBlackDetail {
  emitterUpdateScript [
   &EmitterState(Life Cycle Mode=::NewEnumerator1, Loop Behavior=::NewEnumerator1, Loop Delay=0.4, Loop Duration=3, UseLoopDelay=false)
   &SpawnBurst_Instantaneous(Use Spawn Probability=false, UseLoopCountLimit=false)
   &SpawnRate(
    Spawn Probability=0.7
    SpawnRate=&FloatFromCurve(
     FloatCurve=DataInterface() {propertyValues {Curve {keys [
         {arriveTangent=1.5127, interpMode=RCIM_Cubic, leaveTangent=1.5127, tangentMode=RCTM_User, time=0, value=0.5}
         {interpMode=RCIM_Cubic, tangentMode=RCTM_User, time=1, value=1}
        ]}}}
     CurveIndex=$Emitter.NormalizedLoopAge
     Scale Curve=2
    )
    Use Spawn Probability=true
   )
  ]
  particleSpawnScript [
   &InitializeParticle(
    Color=<rgba 1 1 1 2>
    Lifetime=1.51
    Sprite Rotation Mode=::NewEnumerator2
    Sprite Size Mode=::NewEnumerator1
    Uniform Sprite Size Max=139.6005
    Uniform Sprite Size Min=96.1644
    UsePositionOffset=true
    Write Material Random=true
   )
   &CameraOffset()
   &SkeletalMeshLocation(
    Bone / Socket Index=&SelectIntFromArray(Direct Array Index=&ReturnExecIndex(), Int Selection Array=DataInterface() {propertyValues {IntData [16]}})
    Bone Sampling Mode=::NewEnumerator4
    MaskSpawnGroup=false
    Skeletal Mesh=$System.Skeletal Mesh
   )
   &AddVelocity(
    Rotation Coordinate Space=::World
    Velocity=&RandomRangeVector(Minimum=<0 0 0>, Maximum=<0 0 1>)
    Velocity Speed Scale=&FloatFromCurve(
     FloatCurve=DataInterface() {propertyValues {Curve {keys [{interpMode=RCIM_Cubic, tangentMode=RCTM_User, time=0, value=0.5347}, {interpMode=RCIM_Cubic, tangentMode=RCTM_User, time=1, value=1}]}}}
     CurveIndex=$Emitter.NormalizedLoopAge
     Scale Curve=50
    )
   )
  ]
  particleUpdateScript [
   &ParticleState()
   &SpriteRotationRate(Rotation Rate=&RandomRangeFloat(Minimum=-50, Maximum=50))
   &ScaleSpriteSize(
    Scale Sprite Size Mode=::NewEnumerator0
    Uniform Scale Factor=&FloatFromCurve(FloatCurve=DataInterface() {propertyValues {Curve {keys [<at 0 0.8>,<at 1 1>]}}}, Scale Curve=0.45)
   )
   &Drag(Drag=0.3, Use Linear Drag=true, Use Rotational Drag=true)
   &SubUVAnimation(
    End Frame Range Override=30
    "Sprite Renderer "=DataInterface() {propertyValues {SpriteRenderer=&self:SmokeBlackDetail_6.NiagaraSpriteRendererProperties_0}}
    Start Frame Range Override=&RandomRangeInt(Minimum=0, Maximum=5)
    UseEndFrame=true
    UseStartFrame=true
   )
   &ScaleColor(
    Scale RGB=&MakeVectorFromLinearColorRGB(Color=&ColorFromCurve(ColorCurve=DataInterface() {propertyValues {
        AlphaCurve {keys []}
        BlueCurve {keys [
          <at 0.008 0.0765>
          {arriveTangent=0.2776, interpMode=RCIM_Cubic, leaveTangent=0.2776, time=0.2544, value=0.1234}
          <at 0.9908 0.3493 RCIM_Cubic>
         ]}
        GreenCurve {keys [
          <at 0.008 0.1601>
          {arriveTangent=0.3143, interpMode=RCIM_Cubic, leaveTangent=0.3143, time=0.2544, value=0.1732}
          <at 0.9908 0.469 RCIM_Cubic>
         ]}
        RedCurve {keys [
          <at 0.008 0.1753>
          {arriveTangent=0.3156, interpMode=RCIM_Cubic, leaveTangent=0.3156, time=0.2544, value=0.1753}
          <at 0.9908 0.4855 RCIM_Cubic>
         ]}
       }}))
    ScaleA=true
    ScaleRGB=true
   ) {module=ScaleColor001}
   &ScaleColor(
    Scale Alpha=&Multiply_Float(
     A=&RandomRangeFloat(Minimum=0.5, Maximum=1)
     B=&FloatFromCurve(
      FloatCurve=DataInterface() {propertyValues {Curve {keys [
          {interpMode=RCIM_Cubic, tangentMode=RCTM_User, time=0, value=0}
          {arriveTangent=1.1633, interpMode=RCIM_Cubic, leaveTangent=1.1633, time=0.169, value=0.7909}
          {arriveTangent=0.1774, interpMode=RCIM_Cubic, leaveTangent=0.1774, time=0.1907, value=0.2218}
          {arriveTangent=-0.0424, interpMode=RCIM_Cubic, leaveTangent=-0.0424, time=0.2448, value=0.8044}
          {arriveTangent=0.5088, interpMode=RCIM_Cubic, leaveTangent=0.5088, time=0.2962, value=0.2173}
          {arriveTangent=0.579, interpMode=RCIM_Cubic, leaveTangent=0.579, tangentMode=RCTM_User, time=0.3256, value=0.8455}
          {arriveTangent=0.2517, interpMode=RCIM_Cubic, leaveTangent=0.2517, time=0.3585, value=0.2369}
          {arriveTangent=-0.0276, interpMode=RCIM_Cubic, leaveTangent=-0.0276, time=0.3937, value=0.8627}
          {arriveTangent=0.1287, interpMode=RCIM_Cubic, leaveTangent=0.1287, time=0.4208, value=0.2352}
          {interpMode=RCIM_Cubic, tangentMode=RCTM_User, time=0.4401, value=0.8686}
          {arriveTangent=0.1992, interpMode=RCIM_Cubic, leaveTangent=0.1992, time=0.4533, value=0.2352}
          {arriveTangent=-0.0752, interpMode=RCIM_Cubic, leaveTangent=-0.0752, time=0.4776, value=0.8761}
          {arriveTangent=-0.0752, interpMode=RCIM_Cubic, leaveTangent=-0.0752, time=0.5128, value=0.2307}
          {arriveTangent=0.0613, interpMode=RCIM_Cubic, leaveTangent=0.0613, time=0.5372, value=0.8716}
          {arriveTangent=-0.2207, interpMode=RCIM_Cubic, leaveTangent=-0.2207, time=0.5859, value=0.2352}
          {arriveTangent=-0.5681, interpMode=RCIM_Cubic, leaveTangent=-0.5681, time=0.6184, value=0.8537}
          {tangentMode=RCTM_User, time=1, value=0}
         ]}}}
      Scale Curve=&Multiply_Float(
       A=&FN_FrontEndFloatModifier(Gameplay=1.5, Front End=1)
       B=&FNTimeOfDayScalar(Night Scale=0.5, Morning Scale=0.6, Day Scale=0.6, Evening Scale=0.7)
      )
     )
    )
    Scale RGB=&MakeVectorFromLinearColorRGB(Color=&MakeCustomLinearColorFromBool(
      Bool=$User.bIsFrontEndPreview
      True Color=<rgba 0.7865 0.8826 1 1>
      False Color=&FNTimeOfDayColor(Night Color=<rgba 0.3 0.475 1 1>, Morning Color=<rgba 1 0.8721 0.4868 1>, Day Color=<rgba 1 1 1 1>, Evening Color=<rgba 1 1 1 1>)
     ))
    ScaleA=true
    ScaleRGB=true
   )
   &SolveForcesAndVelocity(Clamp Velocity=false, Limit Acceleration=false)
  ]
  rendererClasses [&NiagaraSpriteRendererProperties]
  renderers [{rendererClass=&NiagaraSpriteRendererProperties}]
 }
 SmokeBlackDetail001 {
  emitterUpdateScript [
   &EmitterState(Life Cycle Mode=::NewEnumerator1, Loop Behavior=::NewEnumerator1, Loop Delay=4.2, Loop Duration=1, UseLoopDelay=true)
   &SpawnBurst_Instantaneous(Spawn Count=3, Use Spawn Probability=false, UseLoopCountLimit=false)
   &SpawnRate(
    SpawnRate=&FloatFromCurve(
     FloatCurve=DataInterface() {propertyValues {Curve {keys [
         {arriveTangent=1.5127, interpMode=RCIM_Cubic, leaveTangent=1.5127, tangentMode=RCTM_User, time=0, value=0.5}
         {interpMode=RCIM_Cubic, tangentMode=RCTM_User, time=1, value=1}
        ]}}}
     CurveIndex=$Emitter.NormalizedLoopAge
     Scale Curve=4
    )
    Use Spawn Probability=true
   )
  ]
  particleSpawnScript [
   &InitializeParticle(
    Color=<rgba 1 1 1 2>
    Lifetime Max=1.3
    Lifetime Min=1.6
    Lifetime Mode=::NewEnumerator1
    Sprite Rotation Mode=::NewEnumerator2
    Sprite Size Mode=::NewEnumerator1
    Uniform Sprite Size Max=139.6005
    Uniform Sprite Size Min=141.9942
    UsePositionOffset=true
    Write Material Random=true
   )
   &CameraOffset()
   &SkeletalMeshLocation(
    Bone / Socket Index=&SelectIntFromArray(
     Array Sampling Mode=::NewEnumerator1
     Direct Array Index=&ReturnExecIndex()
     Int Selection Array=DataInterface() {propertyValues {IntData [3,4,1,2,13]}}
     Random Seed=0
    )
    Bone Sampling Mode=::NewEnumerator4
    MaskSpawnGroup=false
    Skeletal Mesh=$System.Skeletal Mesh
   )
   &AddVelocity(
    Rotation Coordinate Space=::World
    Velocity=&RandomRangeVector(Minimum=<0 0 0.2>, Maximum=<0 0 1>)
    Velocity Speed Scale=&FloatFromCurve(
     FloatCurve=DataInterface() {propertyValues {Curve {keys [{interpMode=RCIM_Cubic, tangentMode=RCTM_User, time=0, value=0.5347}, {interpMode=RCIM_Cubic, tangentMode=RCTM_User, time=1, value=1}]}}}
     CurveIndex=$Emitter.NormalizedLoopAge
     Scale Curve=60
    )
   )
   &ShapeLocation(
    MaskSpawnGroup=false
    Offset=<0 0 -3>
    Offset Coordinate Space=::World
    Offset Mode=::NewEnumerator0
    Override Seed=false
    Sphere Radius=10
    UseOwnerScale=true
   )
  ]
  particleUpdateScript [
   &ParticleState()
   &SpriteRotationRate(Rotation Rate=&RandomRangeFloat(Minimum=-50, Maximum=50))
   &ScaleSpriteSize(
    Scale Sprite Size Mode=::NewEnumerator0
    Uniform Scale Factor=&FloatFromCurve(FloatCurve=DataInterface() {propertyValues {Curve {keys [<at 0 0.8>,<at 1 1>]}}}, Scale Curve=0.45)
   )
   &Drag(Drag=0.3, Use Linear Drag=true, Use Rotational Drag=true)
   &SubUVAnimation(
    End Frame Range Override=30
    "Sprite Renderer "=DataInterface() {propertyValues {SpriteRenderer=&self:SmokeBlackDetail001_0.NiagaraSpriteRendererProperties_0}}
    Start Frame Range Override=&RandomRangeInt(Minimum=0, Maximum=5)
    UseEndFrame=true
    UseStartFrame=true
   )
   &ScaleColor(
    Scale RGB=&MakeVectorFromLinearColorRGB(Color=&ColorFromCurve(ColorCurve=DataInterface() {propertyValues {
        AlphaCurve {keys []}
        BlueCurve {keys [
          <at 0.008 0.1032>
          {arriveTangent=0.2504, interpMode=RCIM_Cubic, leaveTangent=0.2504, time=0.2544, value=0.1535}
          <at 0.9908 0.3493 RCIM_Cubic>
         ]}
        GreenCurve {keys [<at 0.008 0.1329>, {arriveTangent=0.342, interpMode=RCIM_Cubic, leaveTangent=0.342, time=0.2544, value=0.1744}, <at 0.9908 0.469 RCIM_Cubic>]}
        RedCurve {keys [
          <at 0.008 0.1383>
          {arriveTangent=0.3533, interpMode=RCIM_Cubic, leaveTangent=0.3533, time=0.2544, value=0.1753}
          <at 0.9908 0.4855 RCIM_Cubic>
         ]}
       }}))
    ScaleA=true
    ScaleRGB=true
   ) {module=ScaleColor001}
   &ScaleColor(
    Scale Alpha=&Multiply_Float(
     A=&RandomRangeFloat(Minimum=0.5, Maximum=1)
     B=&FloatFromCurve(
      FloatCurve=DataInterface() {propertyValues {Curve {keys [
          {interpMode=RCIM_Cubic, tangentMode=RCTM_User, time=0, value=0}
          {arriveTangent=0.579, interpMode=RCIM_Cubic, leaveTangent=0.579, tangentMode=RCTM_User, time=0.35, value=0.85}
          {interpMode=RCIM_Cubic, tangentMode=RCTM_User, time=0.4997, value=0.9}
          {tangentMode=RCTM_User, time=1, value=0}
         ]}}}
      Scale Curve=&Multiply_Float(
       A=&FN_FrontEndFloatModifier(Gameplay=1.5, Front End=1)
       B=&FNTimeOfDayScalar(Night Scale=0.5, Morning Scale=0.6, Day Scale=0.6, Evening Scale=0.7)
      )
     )
    )
    Scale RGB=&MakeVectorFromLinearColorRGB(Color=&MakeCustomLinearColorFromBool(
      Bool=$User.bIsFrontEndPreview
      True Color=<rgba 0.7865 0.8826 1 1>
      False Color=&FNTimeOfDayColor(Night Color=<rgba 0.3 0.475 1 1>, Morning Color=<rgba 1 0.8721 0.4868 1>, Day Color=<rgba 1 1 1 1>, Evening Color=<rgba 1 1 1 1>)
     ))
    ScaleA=true
    ScaleRGB=true
   )
   &InheritSourceMovement(Applied Position Scale=<0.5 0.5 0>, ApplyForce=false, ApplyPosition=true, ApplyPositionFalloff=false)
   &SolveForcesAndVelocity(Clamp Velocity=false, Limit Acceleration=false)
  ]
  rendererClasses [&NiagaraSpriteRendererProperties]
  renderers [{rendererClass=&NiagaraSpriteRendererProperties}]
 }
 Spike_and_Flare001 {
  emitterSpawnScript [/SetVariables(Emitter.Enabled:NiagaraBool=&SetBoolByFloatComparison(A=$User.IsFollower, B=0, Comparison Type=::NewEnumerator5)) {isSetParameters=true}]
  emitterUpdateScript [
   &EmitterState(Life Cycle Mode=::NewEnumerator1, Loop Behavior=::NewEnumerator1, Loop Duration=3.8, UseLoopDelay=false)
   &SpawnBurst_Instantaneous(Spawn Count=1, Use Spawn Probability=false)
  ]
  particleSpawnScript [
   &InitializeParticle(
    Color=<rgba 1.75 1.6325 0.3895 0.75>
    Lifetime=4.123
    Position Offset=<20 0 0>
    Sprite Size=&Multiply_Vector2D_ByFloat(Float=80, Vector2D=<1 1>)
    Sprite Size Mode=::NewEnumerator4
    UsePositionOffset=true
    Write Material Random=true
   ) {module=InitializeParticle001}
   &DynamicMaterialParameters(
    Index 0 Param 3=&UniformRangedFloat(Minimum=1, Maximum=1.5)
    Index 1 Param 1=0.2
    Index 1 Param 2=1
    Module.Index0Float0Write=true
    Module.Index0Float1Write=true
    Module.Index0Float2Write=true
    Module.Index0Float3Write=true
    Module.Index1Float0Write=true
    Module.Index1Float1Write=true
    Module.Index1Float2Write=true
    Module.Index1Float3Write=true
    Write Parameter Index 1=::NewEnumerator1
   ) {module=DynamicMaterialParameters001}
   &CameraOffset(Camera Offset Amount=20)
  ]
  particleUpdateScript [
   &ParticleState()
   &ScaleSpriteSize(Uniform Curve Sprite Scale=DataInterface() {propertyValues {Curve {keys [
        {arriveTangent=10.3586, interpMode=RCIM_Cubic, leaveTangent=10.3586, tangentMode=RCTM_User, time=0, value=1.8184}
        {arriveTangent=-20.4592, interpMode=RCIM_Cubic, leaveTangent=-20.4592, time=0.015, value=2.461}
        {arriveTangentWeight=0.0357, interpMode=RCIM_Cubic, leaveTangentWeight=0.2891, tangentMode=RCTM_User, time=0.04, value=1}
        <at 1 1 RCIM_Cubic>
       ]}}})
   &FN_Light_Attributes(
    Color=<rgba 1 0.9601 0.4728 1>
    Create Color Attribute=true
    Create Exponent Attribute=true
    Create Position Attribute=true
    Create Radius Attribute=true
    Exponent=1
    Radius=123.8232
    Scale Color=&FloatFromCurve(
     FloatCurve=DataInterface() {propertyValues {Curve {keys [
         <at 0 0 RCIM_Cubic>
         {arriveTangentWeight=0.0581, interpMode=RCIM_Cubic, leaveTangentWeight=0.349, tangentMode=RCTM_User, time=0.0705, value=1}
         {arriveTangent=-0.0405, interpMode=RCIM_Cubic, leaveTangent=-0.0405, tangentMode=RCTM_User, time=0.8929, value=1}
         {arriveTangent=-8.0089, interpMode=RCIM_Cubic, leaveTangent=-8.0089, tangentMode=RCTM_User, time=1, value=0}
        ]}}}
     Scale Curve=4
    )
   )
   &SkeletalMeshLocation(
    Bone / Socket Index=&SelectIntFromArray(Direct Array Index=$Particles.PartitionParticles.Partition, Int Selection Array=DataInterface() {propertyValues {IntData [16]}})
    Bone Sampling Mode=::NewEnumerator4
    MaskSpawnGroup=false
    Sampled Position Offset=<0 0 7>
    Skeletal Mesh=$System.Skeletal Mesh
   )
   &DynamicMaterialParameters(
    Index 0 Param 1=&FloatFromCurve(FloatCurve=DataInterface() {propertyValues {Curve {keys [<at 0.1 0.1 RCIM_Cubic>,<at 0.8 1>]}}})
    Index 0 Param 2=2
    Index 0 Param 4=&FloatFromCurve(FloatCurve=DataInterface() {propertyValues {Curve {keys [<at 0 1>,<at 1 3>]}}})
    Module.Index0Float0Write=true
    Module.Index0Float1Write=true
    Module.Index0Float2Write=true
    Module.Index0Float3Write=true
   )
   /SetVariables(Particles.SpriteFlareColor:LinearColor=&MakeLinearColorFromVectorAndFloat(
     "Vector (RGB)"=&MakeVectorFromLinearColorRGB(Color=<rgba 2 1.7936 0.8438 1>)
     "Float (Alpha)"=&FloatFromCurve(
      FloatCurve=DataInterface() {propertyValues {Curve {keys [
          {tangentMode=RCTM_User, time=-0.0078, value=1}
          <at 0.0037 0.3012>
          <at 0.0117 1.1006>
          <at 0.0254 0.3183>
          <at 0.068 1.4772>
          <at 0.1059 0.4483>
          <at 0.1194 1.222>
          <at 0.1458 0.3696>
          <at 0.173 1.0707>
          <at 0.1894 0.3577>
          <at 0.2114 1.1686>
          <at 0.2395 0.4039>
          <at 0.2657 1.2346>
          <at 0.2878 0.3692>
          <at 0.318 1.3181>
          <at 0.3421 0.2927>
          <at 0.3843 1.2624>
          <at 0.4187 0.2767>
          <at 0.4406 1.1756>
          <at 0.4708 0.3553>
          <at 0.497 1.2833>
          <at 0.5171 0.317>
          <at 0.5452 1.3459>
          <at 0.5855 0.2892>
          <at 0.6056 1.26>
          <at 0.6321 0.3353>
          <at 0.6409 1.2009>
          <at 0.6679 0.2927>
          <at 0.7082 1.0887>
          <at 0.7399 0.1759>
          {arriveTangent=-0.0405, leaveTangent=-0.0405, time=0.7514, value=1.1839}
          <at 0.7659 0.2802>
          <at 0.7801 1.1034>
          <at 0.7922 0.225>
          <at 0.8204 1.0525>
          <at 0.8245 0.1911>
          <at 0.8567 0.9549>
          <at 0.8608 0.1826>
          <at 0.8931 1.0737>
          <at 0.9011 0.1614>
          <at 0.9274 1.0186>
          <at 0.9516 0.1317>
          {arriveTangent=-3.4472, leaveTangent=-3.4472, time=0.9686, value=0.858}
          {arriveTangent=-8.0089, interpMode=RCIM_Cubic, leaveTangent=-8.0089, tangentMode=RCTM_User, time=1, value=0}
         ]}}}
      CurveIndex=&Add_Float(A=$Particles.NormalizedAge, B=&RandomRangeFloat(Minimum=0, Maximum=-0.2))
      Scale Curve=&FNTimeOfDayScalar(Night Scale=0.5, Morning Scale=0.8, Evening Scale=0.8)
     )
    )) {isSetParameters=true}
  ]
  rendererClasses [&NiagaraSpriteRendererProperties, &NiagaraLightRendererProperties]
  renderers [
   {rendererClass=&NiagaraSpriteRendererProperties}
   {rendererClass=&NiagaraLightRendererProperties, rendererIndex=1}
  ]
 }
}
summary {
 systemName=NS_Emote_ShockValue_Zap
 emitters [
  {emitterName=Small_Lightning, simTarget=GPUComputeSim, rendererClasses [&NiagaraMeshRendererProperties]}
  {
   emitterName=Spike_and_Flare001
   rendererClasses [&NiagaraSpriteRendererProperties, &NiagaraLightRendererProperties]
  }
  {emitterName=SmokeBlackDetail, rendererClasses [&NiagaraSpriteRendererProperties]}
  {emitterName=Impact_Spikes, rendererClasses [&NiagaraSpriteRendererProperties]}
  {emitterName=SmokeBlackDetail001, rendererClasses [&NiagaraSpriteRendererProperties]}
  {bEnabled=false, emitterName=Lightning_Small_L002, rendererClasses [&NiagaraSpriteRendererProperties]}
  {emitterName=Lightning_Small_L001, rendererClasses [&NiagaraSpriteRendererProperties]}
 ]
 userVariables [
  User.bIsFrontEnd {type=NiagaraBool}
  User.bIsFrontEndPreview {type=NiagaraBool}
  User.IsFollower {type=NiagaraFloat, defaultValue=0, description=""}
  User.OrnamentOpacity {type=NiagaraFloat, defaultValue=0, description=""}
 ]
}
